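; Quest gameplay ability that, going by its name, tracks the Galileo Lobster "Blocked" feat.
; On event activation it checks the event target for the Item.Weapon.Galileo.Lobster tag, a tag test that replaces a cast to Lobster_Parent.
; On a match it sends a complex custom stat event to the avatar controller's quest manager, with additional target tags and the truncated event magnitude as count, then ends.
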(event Ability|EventActivateAbilityFromEvent
  (params (EventData :type Gameplay%20Event%20Data%20Structure%20%28by%20ref%29))
  (bind (_eventtag _instigator _target _optionalobject _optionalobject2 _contexthandle _instigatortags _targettags _eventmagnitude) (Utilities|Struct|BreakGameplayEventData EventData))
  ;[1 1 1] replacement for a cast to Lobster_Parent (chris.pope)
  (bind _returnvalue (GameplayTags|HasMatchingGameplayTag(Message) _target "(TagName=\"Item.Weapon.Galileo.Lobster\")"))
  (if _returnvalue
    (Quest|SendComplexCustomStatEvent (Quests|GetQuestManager (Fortnite|GetFortPlayerControllerfromActor (Ability|GetAvatarActorfromActorInfo)) "Invalid") _target "(GameplayTags=)" (Variables|Default|GetAdditionalTargetTags) (Math|Float|Truncate _eventmagnitude))
    (else
      _))
  (Ability|EndAbility))
